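# Post-process Anim Blueprint for the StairGummy head skeleton that poses the character's eyelashes with each eye's blink.
# Each update runs Remap Material Blinks on the cached eye dynamic material, which is fetched in the front end, on player mannequins or after customization loads.
# The resulting left and right blink values scrub an eyelash sequence by time.
# Per-bone blends layer those poses onto an input pose with additively rotated eyelash base bones.
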AnimBlueprint /BRCosmetics/Characters/Player/Female/Medium/Heads/F_MED_StairGummy_Head/Meshes/CustomSkeleton/M_MED_StairGummy_Skeleton_AnimBP_Post_Pro
target_skeleton=&/Fortnite_Base_Head_StairGummy_Skeleton
graphs [
 AnimGraph {nodes [
   n1 = "Transform (Modify) Bone - Bone: base_eyelashes_l"(ComponentPose=Local To Component(LocalPose=Input Pose()), Alpha=1, Translation="0,0,0", Rotation="-12,0,5", Scale="1,1,1") {rotationMode=BMM_Additive, translationMode=BMM_Additive, boneToModify {boneName=base_eyelashes_l}, rotation {pitch=-12, roll=5}}
   n2 = "Transform (Modify) Bone - Bone: base_eyelashes_r"(ComponentPose=@n1, Alpha=1, Translation="0,0,0", Rotation="12,0,5", Scale="1,1,1") {rotationMode=BMM_Additive, translationMode=BMM_Additive, boneToModify {boneName=base_eyelashes_r}, rotation {pitch=12, roll=5}}
   n3 = "Sequence Evaluator (by time) 'StairGummy_Eyelashes_Sequence'"!(ExplicitTime:float=Get L Blink Value() -> (L Blink Value:double)) -> (Pose:Pose Link Structure) {
    is=title AnimGraphNode_SequenceEvaluator
    sequence=/BRCosmetics/Characters/Player/Female/Medium/Heads/F_MED_StairGummy_Head/Meshes/StairGummy_EyelashPosing/StairGummy_Eyelashes_Sequence
   }
   n4 = "Sequence Evaluator (by time) 'StairGummy_Eyelashes_Sequence'"(ExplicitTime=Get R Blink Value() -> (R Blink Value:double)) {sequence=&StairGummy_Eyelashes_Sequence}
   n5 = Layered blend per bone!(
    BasePose=Component To Local(ComponentPose=@n2)
    BlendPoses_0=@n3
    BlendPoses_1:Pose Link Structure=@n4
    BlendWeights_0
    BlendWeights_1:float=1
   ) -> (Pose) {
    blendPoses [{linkId=-1, sourceLinkId=-1}, {linkId=-1, sourceLinkId=-1}]
    blendWeights [1,1]
    layerSetup [
     {branchFilters [{boneName=dyn_eyelash_l_04}, {boneName=None}, {boneName=None}]}
     {branchFilters [{boneName=dyn_eyelash_r_04}, {boneName=None}, {boneName=None}]}
    ]
   }
   Output Pose(Result=@n5)
  ]}
 EventGraph {nodes [
   Event BlueprintUpdateAnimation()
   > n1 = Sequence()
   > n2 = Branch(Condition=Get bIsInFrontEnd())
   n3 = Get Dynamic Material Instance(execute=[@n4.then, @n2.then], SkeletalMeshComponent=GetPartSkeletalMeshComponent())
   > Set Eye Mat(Eye Mat:Material Instance Dynamic Object Reference=@n3) -> (Output_Get:Material Instance Dynamic Object Reference)
   n4 = Cast To FortPlayerMannequin!(execute=@n2.else, Object=GetOwningActor()) -> (then, CastFailed, AsFort Player Mannequin:Fort Player Mannequin Object Reference) {is=title K2Node_DynamicCast}
   n5 = Remap Material Blinks!(
    execute=@n1.then_1
    Character Part Anim Instance:Anim Instance Object Reference=Self-Reference()
    Dynamic Material Instance:Material Instance Dynamic Object Reference=Get Eye Mat() -> (Eye Mat:Material Instance Dynamic Object Reference)
    Squint Influence:double=0.35
    "Anim Length(Seconds)":double=3.3
   ) -> (then, L Blink Value:double, R Blink Value:double, L Blink Lower Value:double, R Blink Lower Value:double)
   > Set L Blink Value(L Blink Value:double=@n5) -> (Output_Get:double)
   > Set R Blink Value(R Blink Value:double=@n5.R Blink Value) -> (Output_Get:double)
   K2Node_Event_5 = OnCharacterAndPickaxeCustomizationLoaded_1!() -> (
    OutputDelegate
    then
    CustomBaseMesh:Skeletal Mesh Component Object Reference
    CustomParts:Array of Custom Character Part Object References
    CustomMeshParts:Array of Skeletal Mesh Component Budgeted Object References
   ) {is=title K2Node_CustomEvent}
   > n6 = Get Dynamic Material Instance(SkeletalMeshComponent=GetPartSkeletalMeshComponent())
   > Set Eye Mat(Eye Mat=@n6)
   TryGetPawnOwner()
  ]}
]
variables [
 Eye Mat {type=MaterialInstanceDynamic}
 L Blink Value {type=number}
 R Blink Value {type=number}
]
